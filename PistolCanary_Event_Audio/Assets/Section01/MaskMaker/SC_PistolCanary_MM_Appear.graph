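# SoundCue for the MaskMaker appearing as a portal activates in Section 01 of the PistolCanary event.
# A single wave player feeds the portal-activate appear wave straight to the output, with no randomization or mixing.
# It uses the far MaskMaker attenuation preset for long-range audibility, plays at reduced volume, and routes through the Fort events sound class.
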SoundCue /PistolCanary_Event_Audio/Assets/Section01/MaskMaker/SC_PistolCanary_MM_Appear
root=WavePlayer_0
nodes [{id=WavePlayer_0, properties {soundWaveAssetPtr=/PistolCanary_Event_Audio/Assets/Section01/wav/PistolCanary_Sec_01_Portal_Activate_MM_Appear}}]
properties {
 attenuationSettings=/PistolCanary_Audio/Assets/Settings/ATT_PistolCanary_MaskMaker_Far
 maxDistance=23000
 soundClassObject=&Fort_Events
 volumeMultiplier=0.5
}
