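# Looping Sound Cue for the MaskMaker aura in Section 01 of the PistolCanary event.
# A single looping Wave Player feeds the cue output directly, with raised pitch and slightly reduced volume.
# It plays through the Fort_Events sound class and uses the MaskMaker "Far" attenuation asset to carry over a long range.
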SoundCue /PistolCanary_Event_Audio/Assets/Section01/MaskMaker/SC_PistolCanary_MM_AuraLoop
root=WavePlayer_0
nodes [{
  id=WavePlayer_0
  properties {bLooping=true, soundWaveAssetPtr=/PistolCanary_Event_Audio/Assets/Section01/wav/PistolCanary_Sec_01_MM1000_MM_Aura_Loop}
 }]
properties {
 attenuationSettings=/PistolCanary_Audio/Assets/Settings/ATT_PistolCanary_MaskMaker_Far
 maxDistance=23000
 pitchMultiplier=1.3
 soundClassObject=&Fort_Events
 volumeMultiplier=0.8
}
