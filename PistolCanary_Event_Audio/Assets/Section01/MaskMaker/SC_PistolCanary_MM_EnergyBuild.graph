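# SC_PistolCanary_MM_EnergyBuild is a Sound Cue for the MaskMaker energy build-up sound in Section 01 of the PistolCanary event.
# A single Wave Player node feeds the Section 01 energy build-up wave straight to the output, with no randomization or mixing.
# It routes through the Fort_Events sound class and uses a dedicated MaskMaker energy-build attenuation asset, with a long maximum audible distance.
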SoundCue /PistolCanary_Event_Audio/Assets/Section01/MaskMaker/SC_PistolCanary_MM_EnergyBuild
root=WavePlayer_1
nodes [{id=WavePlayer_1, properties {soundWaveAssetPtr=/PistolCanary_Event_Audio/Assets/Section01/wav/PistolCanary_Sec_01_MM_EnergyBuildiup}}]
properties {
 attenuationSettings=/PistolCanary_Audio/Assets/Settings/ATT_PistolCanary_MaskMaker_EnergyBuild
 maxDistance=55000
 soundClassObject=&Fort_Events
}
