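# Sound cue for the PistolCanary event's Section 01 MaskMaker explosion, apparently triggered by the shoot beam, judging by its name.
# A single wave player plays the portal-explode begin-to-float wave with no randomization or mixing.
# The cue uses the house explosion attenuation with a very long max distance and the Fort_Events sound class.
# It also partially sends to the FortWeapon volume sidechain audio bus, likely to duck weapon sounds.
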SoundCue /PistolCanary_Event_Audio/Assets/Section01/MaskMaker/SC_PistolCanary_MM_Explo_FromShootBeam
root=WavePlayer_1
nodes [{
  id=WavePlayer_1
  properties {soundWaveAssetPtr=/PistolCanary_Event_Audio/Assets/Section01/wav/PistolCanary_Sec_01_Portal_Explode_MM_BeginToFloat}
 }]
properties {
 attenuationSettings=/PistolCanary_Audio/Assets/Settings/ATT_PistolCanary_Explosions_House
 maxDistance=150000
 soundClassObject=&Fort_Events
 volumeMultiplier=1
 busSends [{audioBus=&AB_Sidechain_FortWeapon_Vol, minSendLevel=0, sendLevel=0.6, soundSourceBus=null}]
}
